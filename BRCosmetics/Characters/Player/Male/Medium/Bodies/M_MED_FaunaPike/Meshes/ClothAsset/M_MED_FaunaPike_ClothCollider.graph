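# Physics asset that serves, by its name and path, as the cloth collider for the FaunaPike medium male player body.
# A capsule covers the pelvis, and a large box on the def_slope_warp bone forms a slab, likely a ground plane.
# Both legs have thigh, calf and foot bodies with no shapes listed.
# Only the right leg has constraints, linking foot, calf and thigh with free swing and twist.
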PhysicsAsset /BRCosmetics/Characters/Player/Male/Medium/Bodies/M_MED_FaunaPike/Meshes/ClothAsset/M_MED_FaunaPike_ClothCollider
bodies {
 calf_l []
 calf_r []
 def_slope_warp [{
   shapeType=Box
   extentX=286.2154
   extentY=278.7956
   extentZ=33.447
   shapeName=def_slope_warp_box
   center=<0 0 -14.0088>
   rotation=<rot 0 -90 0>
  }]
 foot_l []
 foot_r []
 pelvis [{shapeType=Capsule, length=8.9786, radius=8.026, center=<0 -1.3781 0>}]
 thigh_l []
 thigh_r []
}
constraints [
 {
  bone1Name=foot_r
  bone2Name=calf_r
  swing1LimitDegrees=45
  swing1Motion=Free
  swing2LimitDegrees=45
  swing2Motion=Free
  twistLimitDegrees=45
  twistMotion=Free
 }
 {
  bone1Name=calf_r
  bone2Name=thigh_r
  swing1LimitDegrees=45
  swing1Motion=Free
  swing2LimitDegrees=45
  swing2Motion=Free
  twistLimitDegrees=45
  twistMotion=Free
 }
]
