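# Chaos Cloth asset that simulates clothing on the male medium FaunaPike character body.
# Merges separate sim and render meshes, adds a backstop accessory mesh, proxy-deforms, then chains constraint, aerodynamic, collision and solver configs.
# Painted maps drive max distance, anim drive and vertex-spring selections; the aero and collision-thickness maps both write "Lift".
# A remeshed LOD1 branch exists but is unconnected to the terminal, which receives only LOD0.
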ChaosClothAsset /BRCosmetics/Characters/Player/Male/Medium/Bodies/M_MED_FaunaPike/Meshes/ClothAsset/M_MED_FaunaPike_Clothing
name=EmbeddedDataflow
graph=&self:EmbeddedDataflow
connections [
 {fromNode=SimulationDampingConfig, fromPinGuid=#1, toNode=AeroMap, toPinGuid=#2}
 {
  fromNode=SimulationMaxDistanceConfig
  fromPin=KinematicVertices3D
  fromPinGuid=#3
  toNode=SkinningBlend
  toPin=KinematicVertices3D.StringValue
  toPinGuid=#4
 }
 {
  fromNode=ReRouteNode_v1
  fromPin=Value
  fromPinGuid=#5
  toNode=ProxyDeformer
  toPin="SelectionFilterSets[0].SimSelection.StringValue"
  toPinGuid=#6
 }
 {
  fromNode=ProceduralSelection_Render
  fromPinGuid=#7
  toNode=ProxyDeformer
  toPinGuid=#8
 }
 {
  fromNode=WeightMapToSelection_0
  fromPin=SelectionName
  fromPinGuid=#9
  toNode=SimulationClothVertexSpringConfig
  toPin="ConstructionSets[0].TargetVertexSelection.StringValue"
  toPinGuid=#10
 }
 {
  fromNode=SkeletalMeshImport_Backstop
  fromPinGuid=#11
  toNode=SimAccessoryMeshNode
  toPin=SimAccessoryMeshCollection
  toPinGuid=#12
 }
 {
  fromNode=SimulationAnimDriveConfig
  fromPinGuid=#13
  toNode=SimulationPressureConfig_0
  toPinGuid=#14
 }
 {
  fromNode=AeroMap
  fromPinGuid=#15
  toNode=SimulationAerodynamicsConfig
  toPinGuid=#16
 }
 {
  fromNode=MergeClothCollections
  fromPinGuid=#17
  toNode=SimAccessoryMeshNode
  toPinGuid=#18
 }
 {
  fromNode=AeroMap
  fromPin=OutputName.StringValue
  fromPinGuid=#19
  toNode=SimulationAerodynamicsConfig
  toPin=Drag.WeightMap
  toPinGuid=#20
 }
 {
  fromNode=SimulationVelocityScaleConfig
  fromPinGuid=#21
  toNode=AnimDriveStiffness
  toPinGuid=#22
 }
 {
  fromNode=SimulationSolverConfig LOD0
  fromPinGuid=#23
  toNode=ReferenceBone
  toPinGuid=#24
 }
 {
  fromNode=WeightMapToSelection
  fromPinGuid=#25
  toNode=TargetVertex Map
  toPinGuid=#26
 }
 {
  fromNode=SimulationCollisionConfig
  fromPinGuid=#27
  toNode=SetPhysicsAsset
  toPinGuid=#28
 }
 {
  fromNode=SimulationPressureConfig_0
  fromPinGuid=#29
  toNode=CollisionThickness_Map
  toPinGuid=#30
 }
 {fromNode=ReferenceBone, fromPinGuid=#31, toNode=StripUserAttributes, toPinGuid=#32}
 {
  fromNode=SimulationMassConfig
  fromPinGuid=#33
  toNode=SimulationLongRangeAttachmentConfig
  toPinGuid=#34
 }
 {
  fromNode=CollisionThickness_Map
  fromPin=OutputName.StringValue
  fromPinGuid=#35
  toNode=SimulationCollisionConfig
  toPin=ClothCollisionThicknessImported.WeightMap
  toPinGuid=#36
 }
 {
  fromNode=SkeletalMeshImport_Sim
  fromPinGuid=#37
  toNode=MergeClothCollections
  toPin="Collections[0]"
  toPinGuid=#38
 }
 {
  fromNode=AeroMap
  fromPin=OutputName.StringValue
  fromPinGuid=#19
  toNode=SimulationAerodynamicsConfig
  toPin=Lift.WeightMap
  toPinGuid=#39
 }
 {
  fromNode=MaxDistance_Mask
  fromPin=OutputName.StringValue
  fromPinGuid=#40
  toNode=SimulationMaxDistanceConfig
  toPin=MaxDistance.WeightMap
  toPinGuid=#41
 }
 {
  fromNode=SimulationAerodynamicsConfig
  fromPinGuid=#42
  toNode=SimulationGravityConfig
  toPinGuid=#43
 }
 {
  fromNode=SimAccessoryMeshNode
  fromPinGuid=#44
  toNode=ProceduralSelection_SIM
  toPinGuid=#45
 }
 {
  fromNode=SimulationStretchConfig
  fromPinGuid=#46
  toNode=SimulationDampingConfig
  toPinGuid=#47
 }
 {
  fromNode=SkeletalMeshImport_Render
  fromPinGuid=#48
  toNode=MergeClothCollections
  toPin="Collections[1]"
  toPinGuid=#49
 }
 {
  fromNode=AnimDriveStiffness
  fromPinGuid=#50
  toNode=SimulationAnimDriveConfig
  toPinGuid=#51
 }
 {
  fromNode=CollisionThickness_Map
  fromPinGuid=#52
  toNode=SimulationBackstopConfig
  toPinGuid=#53
 }
 {
  fromNode=SimulationBendingConfig
  fromPinGuid=#54
  toNode=SimulationStretchConfig
  toPinGuid=#55
 }
 {
  fromNode=SimulationLongRangeAttachmentConfig
  fromPinGuid=#56
  toNode=SimulationBendingConfig
  toPinGuid=#57
 }
 {
  fromNode=SimulationGravityConfig
  fromPinGuid=#58
  toNode=SimulationVelocityScaleConfig
  toPinGuid=#59
 }
 {
  fromNode=AnimDriveStiffness
  fromPin=OutputName.StringValue
  fromPinGuid=#60
  toNode=SimulationAnimDriveConfig
  toPin=AnimDriveStiffness.WeightMap
  toPinGuid=#61
 }
 {
  fromNode=ProceduralSelection_SIM
  fromPinGuid=#62
  toNode=ProceduralSelection_Render
  toPinGuid=#63
 }
 {
  fromNode=SkeletalMesh
  fromPin=SkeletalMesh
  fromPinGuid=#64
  toNode=SkeletalMeshImport_Render
  toPin=SkeletalMesh
  toPinGuid=#65
 }
 {
  fromNode=ProceduralSelection_Render
  fromPin=OutputName
  fromPinGuid=#66
  toNode=ProxyDeformer
  toPin="SelectionFilterSets[0].RenderSelection.StringValue"
  toPinGuid=#67
 }
 {
  fromNode=ProceduralSelection_SIM
  fromPin=OutputName
  fromPinGuid=#68
  toNode=ReRouteNode_v1
  toPin=Value
  toPinGuid=#69
 }
 {
  fromNode=SimulationMaxDistanceConfig
  fromPinGuid=#70
  toNode=SkinningBlend
  toPinGuid=#71
 }
 {fromNode=ProxyDeformer, fromPinGuid=#72, toNode=MaxDistance_Mask, toPinGuid=#73}
 {
  fromNode=MaxDistance_Mask
  fromPinGuid=#74
  toNode=SimulationMaxDistanceConfig
  toPinGuid=#75
 }
 {
  fromNode=SimulationBackstopConfig
  fromPinGuid=#76
  toNode=SimulationCollisionConfig
  toPinGuid=#77
 }
 {
  fromNode=SkinningBlend
  fromPinGuid=#78
  toNode=SimulationMassConfig
  toPinGuid=#79
 }
 {
  fromNode=SimulationClothVertexSpringConfig
  fromPinGuid=#80
  toNode=SimulationSolverConfig LOD0
  toPinGuid=#81
 }
 {fromNode=SetPhysicsAsset, fromPinGuid=#82, toNode=SourceVertex Map, toPinGuid=#83}
 {
  fromNode=SourceVertex Map
  fromPinGuid=#84
  toNode=WeightMapToSelection
  toPinGuid=#85
 }
 {
  fromNode=SourceVertex Map
  fromPin=OutputName.StringValue
  fromPinGuid=#86
  toNode=WeightMapToSelection
  toPin=WeightMapName
  toPinGuid=#87
 }
 {
  fromNode=TargetVertex Map
  fromPinGuid=#88
  toNode=WeightMapToSelection_0
  toPinGuid=#89
 }
 {
  fromNode=TargetVertex Map
  fromPin=OutputName.StringValue
  fromPinGuid=#90
  toNode=WeightMapToSelection_0
  toPin=WeightMapName
  toPinGuid=#91
 }
 {
  fromNode=WeightMapToSelection_0
  fromPinGuid=#92
  toNode=SimulationClothVertexSpringConfig
  toPinGuid=#93
 }
 {
  fromNode=WeightMapToSelection
  fromPin=SelectionName
  fromPinGuid=#94
  toNode=SimulationClothVertexSpringConfig
  toPin="ConstructionSets[0].SourceVertexSelection.StringValue"
  toPinGuid=#95
 }
 {
  fromNode=LOD1_ProceduralSelection_SIM
  fromPin=OutputName
  fromPinGuid=#96
  toNode=LOD1_ProxyDeformer
  toPin="SelectionFilterSets[0].SimSelection.StringValue"
  toPinGuid=#97
 }
 {
  fromNode=DeleteElement
  fromPinGuid=#98
  toNode=LOD1_Density_WeightMap
  toPinGuid=#99
 }
 {fromNode=LOD1_Density_WeightMap, fromPinGuid=#100, toNode=Remesh, toPinGuid=#101}
 {
  fromNode=LOD1_Density_WeightMap
  fromPin=OutputName.StringValue
  fromPinGuid=#102
  toNode=Remesh
  toPin=DensityMapSim.WeightMap
  toPinGuid=#103
 }
 {
  fromNode=LOD1_SkeletalMesh
  fromPin=SkeletalMesh
  fromPinGuid=#104
  toNode=SkeletalMeshImport
  toPin=SkeletalMesh
  toPinGuid=#105
 }
 {
  fromNode=Remesh
  fromPinGuid=#106
  toNode=LOD1_MergeClothCollections
  toPin="Collections[0]"
  toPinGuid=#107
 }
 {
  fromNode=SkeletalMeshImport
  fromPinGuid=#108
  toNode=LOD1_MergeClothCollections
  toPin="Collections[1]"
  toPinGuid=#109
 }
 {
  fromNode=LOD1_ProceduralSelection_SIM
  fromPinGuid=#110
  toNode=LOD1_ProceduralSelection_Render
  toPinGuid=#111
 }
 {
  fromNode=LOD1_ProceduralSelection_Render
  fromPinGuid=#112
  toNode=LOD1_ProxyDeformer
  toPinGuid=#113
 }
 {
  fromNode=LOD1_ProceduralSelection_Render
  fromPin=OutputName
  fromPinGuid=#114
  toNode=LOD1_ProxyDeformer
  toPin="SelectionFilterSets[0].RenderSelection.StringValue"
  toPinGuid=#115
 }
 {
  fromNode=LOD1_MergeClothCollections
  fromPinGuid=#116
  toNode=LOD1_ProceduralSelection_SIM
  toPinGuid=#117
 }
 {
  fromNode=StripUserAttributes
  fromPinGuid=#118
  toNode=ClothAssetTerminal
  toPin="CollectionLods[0]"
  toPinGuid=#119
 }
]
nodes [
 SimulationMaxDistanceConfig {
  type=FChaosClothAssetSimulationMaxDistanceConfigNode
  guid=#120
  properties {MaxDistance=(bIsAnimatable=True,High=250,WeightMap="MaxDistance")}
 }
 SimulationLongRangeAttachmentConfig {
  type=FChaosClothAssetSimulationLongRangeAttachmentConfigNode_v2
  guid=#121
  inputPins [
   Collection {type=FManagedArrayCollection}
   FixedEndSet.StringValue {type=FString}
   TetherStiffness.WeightMap {type=FString}
   TetherScale.WeightMap {type=FString}
   "CustomTetherData[0].CustomFixedEndSet.StringValue" {type=FString}
   "CustomTetherData[0].CustomDynamicEndSet.StringValue" {type=FString}
  ]
  outputPins [Collection {type=FManagedArrayCollection}]
  properties {
   CustomTetherData=(())
   FixedEndSet=(StringValue="KinematicVertices3D")
   TetherScale=(bIsAnimatable=True,Low=1,High=1,WeightMap="TetherScale")
   TetherStiffness=(bIsAnimatable=True,Low=1,High=1,WeightMap="TetherStiffness")
   bEnableCustomTetherGeneration=False
   bUseGeodesicTethers=True
   bWarnDuplicateProperty=True
  }
 }
 SetPhysicsAsset {
  type=FChaosClothAssetSetPhysicsAssetNode
  guid=#122
  properties {PhysicsAsset=/Script/Engine.PhysicsAsset'&/M_MED_FaunaPike_ClothCollider'}
 }
 SimulationStretchConfig {
  type=FChaosClothAssetSimulationStretchConfigNode
  guid=#123
  properties {
   StretchAnisoDamping=(bIsAnimatable=True,Low=5,High=5,WeightMap="StretchAnisoDamping")
   StretchStiffnessBias=(bIsAnimatable=True,Low=200,High=200,WeightMap="StretchStiffnessBias")
   StretchStiffnessWarp=(bIsAnimatable=True,Low=200,High=200,WeightMap="StretchStiffnessWarp")
   StretchStiffnessWeft=(bIsAnimatable=True,Low=200,High=200,WeightMap="StretchStiffnessWeft")
  }
 }
 SimulationBendingConfig {
  type=FChaosClothAssetSimulationBendingConfigNode
  guid=#124
  inputPins [
   Collection {type=FManagedArrayCollection}
   FlatnessRatio.WeightMap {type=FString}
   RestAngle.WeightMap {type=FString}
   BendingStiffness.WeightMap {type=FString}
   BendingStiffnessWarp.WeightMap {type=FString}
   BendingStiffnessWeft.WeightMap {type=FString}
   BendingStiffnessBias.WeightMap {type=FString}
   BendingDamping.WeightMap {type=FString}
   BendingAnisoDamping.WeightMap {type=FString}
   AnisoBucklingRatio.WeightMap {type=FString}
   BucklingStiffness.WeightMap {type=FString}
   BucklingStiffnessWarp.WeightMap {type=FString}
   BucklingStiffnessWeft.WeightMap {type=FString}
   BucklingStiffnessBias.WeightMap {type=FString}
   BucklingRatioWeighted.WeightMap {type=FString}
  ]
  properties {
   AnisoBucklingRatio=(bIsAnimatable=True,Low=0.5,High=0.5,WeightMap="BucklingRatio")
   BendingAnisoDamping=(bIsAnimatable=True,Low=1,High=1,WeightMap="BendingAnisoDamping")
   BendingStiffnessBias=(bIsAnimatable=True,Low=150,High=50,WeightMap="BendingStiffnessBias")
   BendingStiffnessWarp=(bIsAnimatable=True,Low=150,High=150,WeightMap="BendingStiffnessWarp")
   BendingStiffnessWeft=(bIsAnimatable=True,Low=150,High=150,WeightMap="BendingStiffnessWeft")
   BucklingRatioWeighted=(bIsAnimatable=True,WeightMap="BucklingRatio")
   BucklingStiffness=(bIsAnimatable=True,Low=1,High=1,WeightMap="BucklingStiffness")
   BucklingStiffnessBias=(bIsAnimatable=True,Low=50,High=50,WeightMap="BucklingStiffnessBias")
   BucklingStiffnessWarp=(bIsAnimatable=True,Low=50,High=50,WeightMap="BucklingStiffnessWarp")
   BucklingStiffnessWeft=(bIsAnimatable=True,Low=50,High=50,WeightMap="BucklingStiffnessWeft")
   FlatnessRatio=(High=1,WeightMap="FlatnessRatio")
  }
 }
 SimulationCollisionConfig {
  type=FChaosClothAssetSimulationCollisionConfigNode
  guid=#125
  properties {
   ClothCollisionThicknessImported=(bIsAnimatable=True,Low=0.2,High=3,WeightMap="ClothCollisionThickness")
   CollisionThicknessImported=()
   FrictionCoefficientWeighted=(bIsAnimatable=True,Low=0.2,High=0.2,WeightMap="FrictionCoefficient")
   InnerCollisionThickness=(Low=1,WeightMap="InnerCollisionThickness")
   bEnableComplexColliders=False
   bUseCCD=False
   bUsePlanarConstraintForComplexColliders=False
   bUsePlanarConstraintForSimpleColliders=True
   bUseSelfCollisionSubstepsForSkinnedTriangleMeshes=False
  }
 }
 SimulationMassConfig {
  type=FChaosClothAssetSimulationMassConfigNode
  guid=#126
  properties {DensityWeighted=(Low=0.7,High=0.7,WeightMap="Density")}
 }
 SimulationAerodynamicsConfig {
  type=FChaosClothAssetSimulationAerodynamicsConfigNode
  guid=#127
  properties {
   Drag=(bIsAnimatable=True,Low=1,High=4.5,WeightMap="Drag")
   Lift=(bIsAnimatable=True,Low=0.2,High=0.5,WeightMap="Lift")
   TurbulenceRatio=0.5
   WindVelocity=(X=0.3,Y=-0.5,Z=0)
   WindVelocitySpace=ReferenceBoneSpace
  }
 }
 SimulationDampingConfig {
  type=FChaosClothAssetSimulationDampingConfigNode
  guid=#128
  properties {
   DampingCoefficientWeighted=(Low=0.01,High=0.01,WeightMap="DampingCoefficient")
   LocalDampingAngularCoefficient=0.2
   LocalDampingLinearCoefficient=0.1
   LocalDampingSpace=ReferenceBoneSpace
  }
 }
 ClothAssetTerminal {type=FChaosClothAssetTerminalNode_v2, guid=#129}
 SimulationSolverConfig LOD0 {
  type=FChaosClothAssetSimulationSolverConfigNode
  guid=#130
  properties {MaxNumIterations=2, NumIterations=2, NumSubstepsImported=(ImportedValue=1)}
 }
 StripUserAttributes {
  type=FChaosClothStripUserAttributesNode
  guid=#131
  inputPins [
   Collection {type=FManagedArrayCollection}
   ExtraSimWeightMaps {type="TArray<FString>"}
   ExtraSets {type="TArray<FString>"}
   ExtraSimFaceIntMaps {type="TArray<FString>"}
   ExtraRenderWeightMaps {type="TArray<FString>"}
  ]
  outputPins [Collection {type=FManagedArrayCollection}]
  properties {
   ExtraRenderWeightMaps=""
   ExtraSets=""
   ExtraSimFaceIntMaps=""
   ExtraSimWeightMaps=""
   bCopyAllUserAttributesToSimModel=True
   bStripUnusedRenderWeightMaps=True
   bStripUnusedSets=True
   bStripUnusedSimFaceIntMaps=True
   bStripUnusedSimWeightMaps=True
  }
 }
 SimulationBackstopConfig {
  type=FChaosClothAssetSimulationBackstopConfigNode
  guid=#132
  inputPins [
   Collection {type=FManagedArrayCollection}
   BackstopDistance.WeightMap {type=FString}
   BackstopRadius.WeightMap {type=FString}
   BackstopMeshName.StringValue {type=FString}
  ]
  outputPins [Collection {type=FManagedArrayCollection}]
  properties {
   BackstopDistance=(bIsAnimatable=True,High=5,WeightMap="BackstopDistance")
   BackstopMeshName=(StringValue="AccessoryMesh")
   BackstopRadius=(bIsAnimatable=True,Low=70,High=70,WeightMap="BackstopRadius")
   bWarnDuplicateProperty=True
  }
 }
 SimulationAnimDriveConfig {
  type=FChaosClothAssetSimulationAnimDriveConfigNode
  guid=#133
  inputPins [
   Collection {type=FManagedArrayCollection}
   AnimDriveStiffness.WeightMap {type=FString}
   AnimDriveDamping.WeightMap {type=FString}
  ]
  outputPins [Collection {type=FManagedArrayCollection}]
  properties {
   AnimDriveDamping=(bIsAnimatable=True,High=1,WeightMap="AnimDriveDamping")
   AnimDriveStiffness=(bIsAnimatable=True,High=1,WeightMap="AnimDriveStiffness")
   bWarnDuplicateProperty=True
  }
 }
 ProxyDeformer {
  type=FChaosClothAssetProxyDeformerNode_v3
  guid=#134
  properties {InfluenceRadius=5, SelectionFilterSets=(()), bPreserveRenderTangents=True}
 }
 MergeClothCollections {
  type=FChaosClothAssetMergeClothCollectionsNode_v2
  guid=#135
  inputPins [
   "Collections[0]" {type=FManagedArrayCollection}
   "Collections[1]" {type=FManagedArrayCollection}
   "Collections[2]" {type=FManagedArrayCollection}
  ]
 }
 ReferenceBone {
  type=FChaosClothAssetReferenceBoneNode
  guid=#136
  inputPins [Collection {type=FManagedArrayCollection}]
  outputPins [Collection {type=FManagedArrayCollection}]
  properties {CalculateDefaultReferenceBone=(), ReferenceBone=(Name="root")}
 }
 SkeletalMeshImport_Render {
  type=FChaosClothAssetSkeletalMeshImportNode_v2
  guid=#137
  properties {
   SectionIndex=1
   SkeletalMesh=&SkeletalMesh'/BRCosmetics/Characters/Player/Male/Medium/Bodies/M_MED_FaunaPike/Meshes/M_MED_FaunaPike'
   bImportRenderMesh=True
   bImportSimMesh=False
   bImportSingleSection=True
  }
 }
 SimulationVelocityScaleConfig {
  type=FChaosClothAssetSimulationVelocityScaleConfigNode
  guid=#138
  inputPins [Collection {type=FManagedArrayCollection}]
  outputPins [Collection {type=FManagedArrayCollection}]
  properties {
   AngularVelocityScale=0.1
   FictitiousAngularScale=0.2
   FictitiousForcesModel=FullScale
   LinearVelocityScale=(X=0.4,Y=0.4,Z=0.25)
   MaxAngularAcceleration=1
   MaxAngularVelocity=0.5
   MaxLinearAcceleration=(X=1200,Y=1200,Z=1200)
   MaxLinearVelocity=(X=500,Y=500,Z=500)
   MaxVelocityScale=1
   VelocityScaleSpace=ReferenceBoneSpace
   bEnableAngularAccelerationClamping=True
   bEnableAngularVelocityClamping=True
   bEnableLinearAccelerationClamping=True
   bEnableLinearVelocityClamping=True
   bWarnDuplicateProperty=True
  }
 }
 SimulationGravityConfig {
  type=FChaosClothAssetSimulationGravityConfigNode
  guid=#139
  properties {GravityOverrideImported=(ImportedValue=(X=0,Y=0,Z=-980.665))}
 }
 ProceduralSelection_SIM {
  type=FChaosClothAssetProceduralSelectionNode
  guid=#140
  inputPins [
   Collection {type=FManagedArrayCollection}
   ConversionInputName.StringValue {type=FString}
  ]
  outputPins [
   Collection {type=FManagedArrayCollection}
   OutputName {type=FString}
  ]
  properties {ConversionInputName=(), Group=(Name="SimVertices3D"), OutputName=SimSelectionProcedural, SelectionType=SelectAll}
 }
 ProceduralSelection_Render {
  type=FChaosClothAssetProceduralSelectionNode
  guid=#141
  inputPins [
   Collection {type=FManagedArrayCollection}
   ConversionInputName.StringValue {type=FString}
  ]
  outputPins [
   Collection {type=FManagedArrayCollection}
   OutputName {type=FString}
  ]
  properties {ConversionInputName=(), Group=(Name="RenderVertices"), OutputName=RenderSelectionProcedural, SelectionType=SelectAll}
 }
 SkinningBlend {
  type=FChaosClothAssetSkinningBlendNode
  guid=#142
  inputPins [
   Collection {type=FManagedArrayCollection}
   KinematicVertices3D.StringValue {type=FString}
  ]
  outputPins [
   Collection {type=FManagedArrayCollection}
   SkinningBlendName {type=FString}
  ]
  properties {KinematicVertices3D=(), SkinningBlendName=RenderDeformerSkinningBlend, bUseSmoothTransition=True}
 }
 MaxDistance_Mask {
  type=FChaosClothAssetWeightMapNode
  guid=#143
  properties {MapOverrideType=ReplaceAll, OutputName=(StringValue="MaxDistance"), TransferType=Use3DSimMesh}
 }
 AnimDriveStiffness {
  type=FChaosClothAssetWeightMapNode
  guid=#144
  properties {MapOverrideType=ReplaceAll, OutputName=(StringValue="AnimDriveStiffness"), TransferType=Use3DSimMesh}
 }
 ReRouteNode_v1 {type=FDataflowReRouteNode, guid=#145, inputPins [Value {type=FString}], outputPins [Value {type=FString}]}
 SkeletalMeshImport_Sim {
  type=FChaosClothAssetSkeletalMeshImportNode_v2
  guid=#146
  properties {
   SkeletalMesh=&SkeletalMesh'&/M_MED_FaunaPike_Simulation'
   bImportRenderMesh=False
   bImportSimMesh=True
   bImportSingleSection=False
  }
 }
 SkeletalMesh {
  type=FGetSkeletalMeshDataflowNode
  guid=#147
  inputPins []
  outputPins [SkeletalMesh {type="TObjectPtr<USkeletalMesh>"}]
  properties {
   PropertyName=SkeletalMesh
   SkeletalMesh=&SkeletalMesh'&M_MED_FaunaPike'
  }
 }
 SkeletalMeshImport_Backstop {
  type=FChaosClothAssetSkeletalMeshImportNode_v2
  guid=#148
  properties {
   SkeletalMesh=&SkeletalMesh'&/M_MED_FaunaPike_Backstop'
   bImportRenderMesh=False
   bImportSimMesh=True
   bImportSingleSection=True
  }
 }
 SimAccessoryMeshNode {
  type=FChaosClothAssetSimAccessoryMeshNode
  guid=#149
  inputPins [
   Collection {type=FManagedArrayCollection}
   SimAccessoryMeshCollection {type=FManagedArrayCollection}
  ]
  outputPins [
   Collection {type=FManagedArrayCollection}
   AccessoryMeshName {type=FString}
  ]
  properties {AccessoryMeshName=AccessoryMesh, bUseSimImportVertexID=True}
 }
 CollisionThickness_Map {
  type=FChaosClothAssetWeightMapNode
  guid=#150
  properties {MapOverrideType=ReplaceAll, OutputName=(StringValue="Lift"), TransferType=Use3DSimMesh}
 }
 AeroMap {
  type=FChaosClothAssetWeightMapNode
  guid=#151
  properties {MapOverrideType=ReplaceAll, OutputName=(StringValue="Lift"), TransferType=Use3DSimMesh}
 }
 SimulationClothVertexSpringConfig {
  type=FChaosClothAssetSimulationClothVertexSpringConfigNode
  guid=#152
  inputPins [
   Collection {type=FManagedArrayCollection}
   "ConstructionSets[0].SourceVertexSelection.StringValue" {type=FString}
   "ConstructionSets[0].TargetVertexSelection.StringValue" {type=FString}
  ]
  outputPins [Collection {type=FManagedArrayCollection}]
  properties {
   ConstraintVertices=((X=15,Y=44),(X=20,Y=33),(X=211,Y=213),(X=217,Y=219))
   ConstructionSets=(())
   GenerateConstraints=()
   RestLengthScale=1
   RestLengths=(1.7769,1.2969,1.2969,1.7769)
   VertexSpringCompressionStiffness=(X=5,Y=5)
   VertexSpringDamping=(X=0.45,Y=0.45)
   VertexSpringExtensionStiffness=(X=2.5,Y=2.5)
   bAppendToExisting=False
   bWarnDuplicateProperty=True
  }
 }
 WeightMapToSelection {
  type=FChaosClothAssetWeightMapToSelectionNode
  guid=#153
  inputPins [
   Collection {type=FManagedArrayCollection}
   WeightMapName {type=FString}
  ]
  outputPins [
   Collection {type=FManagedArrayCollection}
   SelectionName {type=FString}
  ]
  properties {SelectionName="", SelectionThreshold=0.95, SelectionType=SimVertices3D}
 }
 SourceVertex Map {type=FChaosClothAssetWeightMapNode, guid=#154, properties {OutputName=(StringValue="SourceVertConstraint")}}
 WeightMapToSelection_0 {
  type=FChaosClothAssetWeightMapToSelectionNode
  guid=#155
  inputPins [
   Collection {type=FManagedArrayCollection}
   WeightMapName {type=FString}
  ]
  outputPins [
   Collection {type=FManagedArrayCollection}
   SelectionName {type=FString}
  ]
  properties {SelectionName="", SelectionThreshold=0.95, SelectionType=SimVertices3D}
 }
 TargetVertex Map {type=FChaosClothAssetWeightMapNode, guid=#156, properties {OutputName=(StringValue="TargetVertConstraint")}}
 SimulationPressureConfig_0 {
  type=FChaosClothAssetSimulationPressureConfigNode
  guid=#157
  inputPins [
   Collection {type=FManagedArrayCollection}
   Pressure.WeightMap {type=FString}
  ]
  outputPins [Collection {type=FManagedArrayCollection}]
  properties {Pressure=(bIsAnimatable=True,Low=0.4,High=1,WeightMap="Pressure"), bWarnDuplicateProperty=True}
 }
 DeleteElement {
  type=FChaosClothAssetDeleteElementNode
  guid=#158
  inputPins [
   Collection {type=FManagedArrayCollection}
   SelectionName.StringValue {type=FString}
  ]
  outputPins [Collection {type=FManagedArrayCollection}]
  properties {Elements="", Group=(Name="RenderVertices"), SelectionName=(), bDeleteRenderMesh=True, bDeleteSimMesh=False}
 }
 LOD1_Density_WeightMap {type=FChaosClothAssetWeightMapNode, guid=#159, properties {OutputName=()}}
 Remesh {
  type=FChaosClothAssetRemeshNode_v2
  guid=#160
  inputPins [
   Collection {type=FManagedArrayCollection}
   TriangleSelectionSim.StringValue {type=FString}
   DensityMapSim.WeightMap {type=FString}
   TriangleSelectionRender.StringValue {type=FString}
   DensityMapRender.WeightMap {type=FString}
  ]
  outputPins [Collection {type=FManagedArrayCollection}]
  properties {
   DensityMapRender=(Low=100,High=200,WeightMap="DensityMapRender")
   DensityMapSim=(Low=50,High=100,WeightMap="DensityMapSim")
   DensityMapThresholdRender=1
   DensityMapThresholdSim=0.8
   IterationsRender=10
   IterationsSim=10
   RemeshMethodRender=Remesh
   RenderSeamRemeshIterations=1
   SelectionSourceRender=DensityMapInput
   SelectionSourceSim=DensityMapInput
   SmoothingRender=0.25
   SmoothingSim=0.5
   TargetPercentRender=100
   TriangleSelectionRender=()
   TriangleSelectionSim=()
   bRemeshRender=False
   bRemeshRenderSeams=False
   bRemeshSim=True
  }
 }
 LOD1_SkeletalMesh {
  type=FGetSkeletalMeshDataflowNode
  guid=#161
  inputPins []
  outputPins [SkeletalMesh {type="TObjectPtr<USkeletalMesh>"}]
  properties {
   PropertyName=SkeletalMesh
   SkeletalMesh=&SkeletalMesh'&M_MED_FaunaPike'
  }
 }
 SkeletalMeshImport {
  type=FChaosClothAssetSkeletalMeshImportNode_v2
  guid=#162
  properties {LODIndex=1, SectionIndex=1, SkeletalMesh=None, bImportRenderMesh=True, bImportSimMesh=False, bImportSingleSection=True}
 }
 LOD1_MergeClothCollections {type=FChaosClothAssetMergeClothCollectionsNode_v2, guid=#163}
 LOD1_ProceduralSelection_SIM {
  type=FChaosClothAssetProceduralSelectionNode
  guid=#164
  inputPins [
   Collection {type=FManagedArrayCollection}
   ConversionInputName.StringValue {type=FString}
  ]
  outputPins [
   Collection {type=FManagedArrayCollection}
   OutputName {type=FString}
  ]
  properties {ConversionInputName=(), Group=(Name="SimVertices3D"), OutputName=SimSelectionProcedural, SelectionType=SelectAll}
 }
 LOD1_ProceduralSelection_Render {
  type=FChaosClothAssetProceduralSelectionNode
  guid=#165
  inputPins [
   Collection {type=FManagedArrayCollection}
   ConversionInputName.StringValue {type=FString}
  ]
  outputPins [
   Collection {type=FManagedArrayCollection}
   OutputName {type=FString}
  ]
  properties {ConversionInputName=(), Group=(Name="RenderVertices"), OutputName=RenderSelectionProcedural, SelectionType=SelectAll}
 }
 LOD1_ProxyDeformer {
  type=FChaosClothAssetProxyDeformerNode_v3
  guid=#166
  properties {InfluenceRadius=5, SelectionFilterSets=(()), bPreserveRenderTangents=True}
 }
]
